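# Translucent, two-sided, lit master material for Niagara mesh ribbon trails, driven by three sets of dynamic parameters for panning, width, UV scale and erosion.
# SplineThicken sets the trail width, with distorted, panning noise added as local-space vertex displacement.
# Opacity combines a scrolling tile texture, an erodable smoke mask, a sine edge fade and particle alpha, then optional Fresnel and depth fade.
# Dimmed particle color feeds an emissive adjustment function.
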Material /Game/Effects/Niagara/MasterMaterials/MeshTrail/M_Buttercake_MeshRibbon
graph {
 properties {
  bUsedWithNiagaraMeshParticles=true
  bUsedWithNiagaraSprites=true
  blendMode=BLEND_Translucent
  pixelDepthOffsetMode=PDOM_Legacy
  twoSided=true
  usedShadingModels=Default Lit
 }
 nodes [
  n1 = DynamicParameter() {defaultValue=<rgba 1 1 1 0>, paramNames [Panner,Width,WPO,PanSpeedX]}
  n2 = DynamicParameter() {parameterIndex=2, defaultValue=<rgba 0 1 1 0.5>, paramNames [MaskErosion, MaskMultiply, SubTexMultiply, WPOTipMulti]}
  n3 = Constant3Vector() {constant=<rgba 1.5 1.5 0 0>}
  n4 = TextureCoordinate() {uTiling=0.4}
  n5 = ComponentMask(None=TextureSample(UVs=Panner(Coordinate=@n4) {speedX=0.2, speedY=-1.5}) {samplerType=SAMPLERTYPE_Normal, texture=&T_Weap_Distort_N_02}.RGB) {g=true, r=true}
  n6 = Panner(Coordinate=Add(A=TextureCoordinate() {uTiling=0.5}, B=Multiply(A=@n5) {constB=0})) {speedX=0.5, speedY=-2}
  n7 = Power(Base=ConstantBiasScale(None=TextureSample(UVs=@n6) {samplerType=SAMPLERTYPE_Grayscale, texture=&LowResBlurredNoise}.R) {bias=0, scale=2})
  n8 = Multiply(A=Multiply(A=Multiply(A=@n3, B=@n7)), B=DynamicParameter() {defaultValue=<rgba 1 1 1 0>, paramNames [Panner,Width,WPO,PanSpeedX]}.WPO)
  Add(
   A=MaterialFunctionCall(WidthBase=@n1.Width, WidthTip=Multiply(A=@n1.Width, B=@n2.WPOTipMulti) {constB=0.5}) {materialFunction=&SplineThicken}.WorldPositionOffset
   B=Multiply(A=Transform(None=@n8) {transformSourceType=TRANSFORMSOURCE_Local})
  )
  n9 = ParticleColor()
  n10 = Multiply(A=@n9.RGB) {constB=0.3}
  MaterialFunctionCall(Emissive=@n10) {materialFunction=&MF_FX_EmissiveAdjustments}
  n11 = DynamicParameter() {parameterIndex=1, defaultValue=<rgba 1 0 100 1>, paramNames [SubTexUVScale, SubTexUVOffset, DepthFadeDistance, ScaleMaskY]}
  n12 = TextureCoordinate()
  n13 = DynamicParameter() {parameterIndex=1, defaultValue=<rgba 1 0 100 1>, paramNames [SubTexUVScale, SubTexUVOffset, DepthFadeDistance, ScaleMaskY]}
  n14 = Panner(
   Coordinate=MaterialFunctionCall(UVs=Add(A=@n12, B=@n13.SubTexUVOffset), Texture Scale=@n13.SubTexUVScale) {materialFunction=&ScaleUVsByCenter}.UVs
   Speed=AppendVector(A=Constant(), B=DynamicParameter() {defaultValue=<rgba 1 1 1 0>, paramNames [Panner,Width,WPO,PanSpeedX]}.Panner)
  ) {speedX=-2}
  n15 = TextureSample(UVs=@n14) {texture=&T_Smooth_tile}
  n16 = DynamicParameter() {parameterIndex=2, defaultValue=<rgba 0 1 1 0.5>, paramNames [MaskErosion, MaskMultiply, SubTexMultiply, WPOTipMulti]}
  n17 = Multiply(A=@n15.RGB, B=@n16.SubTexMultiply) {constB=5}
  n18 = Constant()
  n19 = DynamicParameter() {parameterIndex=1, defaultValue=<rgba 1 0 100 1>, paramNames [SubTexUVScale, SubTexUVOffset, DepthFadeDistance, ScaleMaskY]}
  n20 = MaterialFunctionCall(
   UVs=Add(A=TextureCoordinate(), B=AppendVector(A=@n18, B=@n19.SubTexUVOffset))
   Texture Scale=AppendVector(A=Constant() {r=1}, B=@n19.ScaleMaskY)
  ) {materialFunction=&ScaleUVsByCenter}
  n21 = DynamicParameter() {defaultValue=<rgba 1 1 1 0>, paramNames [Panner,Width,WPO,PanSpeedX]}
  n22 = Constant()
  n23 = TextureSample(UVs=Panner(
    Coordinate=MaterialFunctionCall(XY=@n20.UVs) {materialFunction=&Swizzle}.YX
    Speed=AppendVector(A=@n21.PanSpeedX, B=@n22)
   ) {speedX=-2}) {samplerType=SAMPLERTYPE_Grayscale, texture=&T_Smoke_Trail_Soft}
  n24 = DynamicParameter() {parameterIndex=2, defaultValue=<rgba 0 1 1 0.5>, paramNames [MaskErosion, MaskMultiply, SubTexMultiply, WPOTipMulti]}
  n25 = Multiply(A=@n17, B=Subtract(A=Multiply(A=@n23.RGB, B=@n24.MaskMultiply) {constB=2}, B=@n24.MaskErosion) {constB=0})
  n26 = TextureCoordinate()
  n27 = Power(Base=Sine(None=ComponentMask(None=@n26) {g=true}) {period=2})
  n28 = Multiply(A=@n9.A, B=Multiply(A=Saturate(None=@n25), B=Saturate(None=@n27)))
  n29 = MaterialFunctionCall(Exponent=ScalarParameter() {defaultValue=0.7, parameterName=FresnelPower}) {materialFunction=&MF_Fresnel_FX}
  Multiply(
   A=LinearInterpolate(A=@n28, B=Multiply(A=@n28, B=@n29.Fresnel), Alpha=ScalarParameter() {parameterName=UseFresnel})
   B=DepthFade(FadeDistance=@n11.DepthFadeDistance)
  )
  Multiply(A=@n15.RGB, B=@n17) {constB=5}
  Power(Base=@n15.RGB, Exp=@n16.SubTexMultiply) {constExponent=0.75}
  Reroute(None=@n9.RGB)
  Reroute(None=@n10)
  ScalarParameter() {defaultValue=200, parameterName=DepthFadeDistance}
  TextureSample() {texture=/Game/Effects/Fort_Effects/Textures/Plasma/T_Turbulence_Stretched}
  TextureSample() {samplerType=SAMPLERTYPE_Grayscale, texture=&LowResBlurredNoise}
  TextureSample() {texture=/Game/Weapons/FORT_Melee/Pickaxe_Cube_Ninja_Male/FX/T_Offset_Gradient_01}
  TextureSample(UVs=@n14) {texture=/Game/Global/Textures/Default/Caustics/T_Turbulence_Seamless}
  TextureSample() {texture=/Game/Global/Textures/Default/Noise/T_Tiling_Noise_01}
 ]
}
