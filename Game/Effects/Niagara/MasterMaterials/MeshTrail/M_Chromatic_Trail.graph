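# Unlit, two-sided translucent master material for chromatic Niagara trail effects on meshes, ribbons and sprites.
# Noise-distorted flow, rainbow and mask textures build the emissive, with optional curve-atlas tint and time-of-day grading.
# World position offset combines normal-based sine waves, tail bending driven by Niagara dynamic parameters, and optional wobble.
# Opacity layers masks with depth, camera and view-angle fades; refraction and noise normals drop out on Low/Medium quality, forward and mobile.
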Material /Game/Effects/Niagara/MasterMaterials/MeshTrail/M_Chromatic_Trail
graph {
 properties {
  bUseTranslucencyVertexFog=false
  bUsedWithNiagaraMeshParticles=true
  bUsedWithNiagaraRibbons=true
  bUsedWithNiagaraSprites=true
  blendMode=BLEND_Translucent
  pixelDepthOffsetMode=PDOM_Legacy
  refractionMethod=RM_IndexOfRefraction
  shadingModel=MSM_Unlit
  twoSided=true
  usedShadingModels=Unlit
 }
 nodes [
  n1 = VertexNormalWS()
  n2 = DynamicParameter() {defaultValue=<rgba 0 0 1 0>, paramNames [UVTilingRandom, TimeOffsetRandom, WPOStrength, TrailingModeON]}
  n3 = Add(A=Time(), B=@n2.TimeOffsetRandom)
  n4 = VectorParameter() {group=WPO, parameterName=WPO Sine, channelNames {a="", b=Sine Strength, g=Sine Tile, r=Sine Speed}, defaultValue=<rgba 2 2 1 1>}
  n5 = DynamicParameter() {parameterIndex=2, defaultValue=<rgba 0 0 1 0>, paramNames [Sine Speed, Sine Tile, Param3, Param4]}
  n6 = Add(A=@n4.Sine Speed, B=LinearInterpolate(A=@n5.Sine Speed, Alpha=ScalarParameter() {parameterName=FEPositionFix}) {constB=1.3})
  n7 = Add(A=@n2.UVTilingRandom, B=ComponentMask(None=TextureCoordinate()) {g=true})
  n8 = Add(A=@n4.Sine Tile, B=LinearInterpolate(A=@n5.Sine Tile, Alpha=ScalarParameter() {parameterName=FEPositionFix}) {constB=1.5})
  n9 = Sine(None=Add(A=Multiply(A=@n3, B=@n6), B=Multiply(A=@n7, B=@n8) {constB=2}))
  n10 = TextureCoordinate()
  n11 = LinearInterpolate(
   A=Constant3Vector()
   B=Multiply(
    A=MaterialFunctionCall(X=@n9, Y=@n9, Z=@n9) {materialFunction=&MakeFloat3}.Result
    B=@n4.Sine Strength
   ) {constB=0.2}
   Alpha=OneMinus(None=ComponentMask(None=@n10) {g=true})
  )
  n12 = DynamicParameter() {defaultValue=<rgba 0 0 1 0>, paramNames [UVTilingRandom, TimeOffsetRandom, WPOStrength, TrailingModeON]}
  n13 = Add(A=Multiply(A=Multiply(A=@n1, B=@n11), B=@n12.WPOStrength) {constB=10})
  n14 = DynamicParameter() {desc=Parameters set by Niagara, parameterIndex=1, paramNames [X,Y,Z,WPOIntensity]}
  n15 = MaterialFunctionCall(X=@n14.X, Y=@n14.X, Z=@n14.X) {materialFunction=&MakeFloat3}
  n16 = DynamicParameter() {desc=Parameters set by Niagara, parameterIndex=2, paramNames [X, Y, Z, ""]}
  n17 = LinearInterpolate(
   A=ObjectPositionWS()
   B=MaterialFunctionCall(X=@n16.X, Y=@n16.X, Z=@n16.X) {materialFunction=&MakeFloat3}.Result
   Alpha=ScalarParameter() {parameterName=FEPositionFix}
  )
  n18 = VectorParameter() {group=WPO, parameterName=WPO Control, channelNames {a=Speed, b=BendPower, g=Max, r=Min}, defaultValue=<rgba -50 50 3 3>}
  n19 = OneMinus(None=ComponentMask(None=TextureCoordinate()) {g=true})
  n20 = LinearInterpolate(
   B=Clamp(None=Subtract(A=@n15.Result, B=@n17), Min=@n18.Min, Max=@n18.Min) {maxDefault=50, minDefault=-50}
   Alpha=Multiply(A=Power(Base=@n19, Exp=@n18.BendPower), B=@n14.WPOIntensity)
  )
  n21 = Constant()
  n22 = Multiply(
   A=MaterialFunctionCall(Speed=ScalarParameter() {defaultValue=0.3, parameterName=WobbleSpeed}) {materialFunction=&DebugTimeSine}.Result
   B=ScalarParameter() {defaultValue=15, parameterName=WobbleAmount}
  )
  Add(
   A=LinearInterpolate(A=@n13, B=Add(A=@n20, B=@n13), Alpha=@n12.TrailingModeON)
   B=LinearInterpolate(
    B=MaterialFunctionCall(X=@n21, Y=@n21, Z=@n22) {materialFunction=&MakeFloat3}.Result
    Alpha=ScalarParameter() {parameterName=DoWobble}
   ) {constB=0}
  )
  Constant()
  n23 = VectorParameter() {group=Flow, parameterName=Flow UV, defaultValue=<rgba 1 1 0 3>}
  n24 = DynamicParameter() {defaultValue=<rgba 0 0 1 0>, paramNames [UVTilingRandom, TimeOffsetRandom, WPOStrength, TrailingModeON]}
  n25 = Multiply(
   A=TextureCoordinate()
   B=MaterialFunctionCall(X=@n23.R, Y=Add(A=@n24.UVTilingRandom, B=@n23.G)) {materialFunction=&MakeFloat2}.Result
  )
  n26 = Multiply(
   A=MaterialFunctionCall(X=@n23.B, Y=@n23.B) {materialFunction=&MakeFloat2}.Result
   B=Add(A=@n24.TimeOffsetRandom, B=Time())
  )
  n27 = TextureCoordinate()
  n28 = VectorParameter() {group=Noise, parameterName=Noise UV, defaultValue=<rgba 1 0.5 0 -0.2>}
  n29 = DynamicParameter() {defaultValue=<rgba 0 0 1 0>, paramNames [UVTilingRandom, TimeOffsetRandom, WPOStrength, TrailingModeON]}
  n30 = MaterialFunctionCall(X=@n28.R, Y=Add(A=@n29.UVTilingRandom, B=@n28.G)) {materialFunction=&MakeFloat2}
  n31 = MaterialFunctionCall(X=@n28.B, Y=@n28.B) {materialFunction=&MakeFloat2}
  n32 = Add(A=@n29.TimeOffsetRandom, B=Time())
  n33 = TextureSampleParameter2D(UVs=Add(A=Multiply(A=@n27, B=@n30.Result), B=Multiply(A=@n31.Result, B=@n32))) {
   group=Noise
   parameterName=Noise Texture
   samplerType=SAMPLERTYPE_Normal
   texture=&T_Weap_Distort_N_02
  }
  n34 = VectorParameter() {parameterName=Base, channelNames {a=NoiseStrength, b=MaskMultiply, g=DepthFade, r=Refraction}, defaultValue=<rgba 0 50 1 0.2>}
  n35 = Multiply(A=ComponentMask(None=@n33.RGB) {g=true, r=true}, B=@n34.NoiseStrength)
  n36 = ScalarParameter() {parameterName=Detial Blur}
  n37 = TextureSampleParameter2D!(UVs=Add(A=Add(A=@n25, B=@n26), B=@n35), Level=@n36, Apply View MipBias) {
   group=Flow
   mipValueMode=TMVM_MipLevel
   parameterName=Flow Texture
   samplerType=SAMPLERTYPE_LinearColor
   texture=/Game/Effects/Fort_Effects/Textures/T_WavePattern_Chromatic
  }
  n38 = VectorParameter() {group=Flow, parameterName=Flow Refine, defaultValue=<rgba 1 1 1 0.5>}
  n39 = Multiply(A=Power(Base=@n37.A, Exp=@n38.R), B=@n38.G)
  n40 = Multiply(A=@n37.A, B=@n38.B)
  n41 = Desaturation(None=@n37.RGB, Fraction=ScalarParameter() {defaultValue=-10, group=Flow, parameterName=Flow Desaturation})
  n42 = VectorParameter() {group=Flow, parameterName=Flow Color, channelNames {a=Add, b="", g="", r=""}, defaultValue=<rgba 1 1 1 0>}
  n43 = Multiply(A=Multiply(A=LinearInterpolate(A=@n39, B=@n40, Alpha=@n38.A), B=Add(A=@n41, B=@n42.Add)), B=@n42)
  n44 = VectorParameter() {group=Color, parameterName=Color UV, defaultValue=<rgba 1 1 0 0>}
  n45 = DynamicParameter() {defaultValue=<rgba 0 0 1 0>, paramNames [UVTilingRandom, TimeOffsetRandom, WPOStrength, TrailingModeON]}
  n46 = Multiply(
   A=TextureCoordinate()
   B=MaterialFunctionCall(X=@n44.R, Y=Add(A=@n45.UVTilingRandom, B=@n44.G)) {materialFunction=&MakeFloat2}.Result
  )
  n47 = Multiply(
   A=MaterialFunctionCall(X=@n44.B, Y=@n44.B) {materialFunction=&MakeFloat2}.Result
   B=Add(A=@n45.TimeOffsetRandom, B=Time())
  )
  n48 = TextureSampleParameter2D(UVs=Add(A=Add(A=@n46, B=@n47), B=@n35), Level=@n36) {group=Color, mipValueMode=TMVM_MipLevel, parameterName=Color Tex, texture=&T_Tile_RainbowColor}
  n49 = Desaturation(None=@n48.RGB, Fraction=ScalarParameter() {group=Color, parameterName=Color Tex Desaturation})
  n50 = VectorParameter() {group=Flow, parameterName=Inside Color, channelNames {a=SizePower, b="", g="", r=""}, defaultValue=<rgba 1 1 1 1>}
  n51 = TextureCoordinate()
  n52 = Multiply(A=@n35, B=Constant2Vector() {r=1})
  n53 = MaterialFunctionCall(Float2=TextureCoordinate()) {materialFunction=&BreakOutFloat2Components}
  n54 = DynamicParameter() {defaultValue=<rgba 0 0 1 0>, paramNames [UVTilingRandom, TimeOffsetRandom, WPOStrength, TrailingModeON]}
  n55 = VectorParameter() {group=Mask, parameterName=Mask UV, defaultValue=<rgba 0.2 1 0.5 0>}
  n56 = Multiply(
   A=MaterialFunctionCall(X=@n53.G, Y=@n53.G) {materialFunction=&MakeFloat2}.Result
   B=MaterialFunctionCall(X=Add(A=@n54.UVTilingRandom, B=@n55.R), Y=@n55.G) {materialFunction=&MakeFloat2}.Result
  )
  n57 = Multiply(
   A=MaterialFunctionCall(X=@n55.B, Y=@n55.B) {materialFunction=&MakeFloat2}.Result
   B=Time()
  )
  n58 = TextureSampleParameter2D!(UVs=Add(A=@n52, B=Add(A=@n56, B=@n57)), Apply View MipBias) {
   group=Mask
   parameterName=Mask Texture
   samplerType=SAMPLERTYPE_Grayscale
   texture=/Game/Effects/Fort_Effects/Textures/Smoke/T_Smoke_Trail_Soft_NoClamp
  }
  n59 = Multiply(A=Saturate(None=ComponentMask(None=@n51) {g=true}), B=@n58.R)
  n60 = LinearInterpolate(
   A=LinearInterpolate(A=@n43, B=Multiply(A=@n49, B=@n43), Alpha=ScalarParameter() {group=Color, parameterName=Use Color Tex})
   B=@n50
   Alpha=Saturate(None=Power(Base=@n59, Exp=@n50.SizePower))
  )
  n61 = CurveAtlasRowParameter(CurveTime=ComponentMask(None=TextureCoordinate()) {g=true}) {
   atlas=/Game/Athena/LinearColorCurves/VFX_CurveAtlas
   curve=/Game/Accessories/FORT_Backpacks/Backpack_F_MED_Cube_Queen/FX/Curve_Backpack_Cube_Queen_Flow_1
   defaultValue=36
   group=Curve
   parameterName=Curve
  }
  n62 = ParticleColor()
  n63 = Multiply(A=LinearInterpolate(A=@n60, B=Multiply(A=@n61, B=@n60), Alpha=ScalarParameter() {group=Curve, parameterName=Use Curve}), B=@n62.RGB)
  n64 = VectorParameter() {group=ColorTimeOfDay, parameterName=ColorTimeOfDay, channelNames {a=Night, b=Evening, g=Day, r=Morning}, defaultValue=<rgba 1 1 1 1>}
  n65 = Multiply(A=@n64.Night, B=@n63)
  n66 = ScalarParameter() {group=ColorTimeOfDay, parameterName=DesaturationNight}
  MaterialFunctionCall(Emissive=MaterialFunctionCall(
    Color_Morning=Multiply(A=@n63, B=@n64.Morning)
    Color_Day=Multiply(A=@n63, B=@n64.Day)
    Color_Evening=Multiply(A=@n63, B=@n64.Evening)
    Color_Night=Desaturation(None=@n65, Fraction=@n66)
   ) {materialFunction=&Color_Time_of_Day}.Result) {materialFunction=&MF_FX_EmissiveAdjustments}
  NamedRerouteDeclaration(None=@n37.A) {name=FlowTextureDesat, nodeColor=<rgba 0.0078 0 1 1>}
  n67 = ScalarParameter() {defaultValue=65, parameterName=WorldPosition_SphereMaskRadius}
  n68 = DynamicParameter() {parameterIndex=3, paramNames [WorldPosition_SphereMaskRadius control, Param2, Param3, AppearMask]}
  NamedRerouteDeclaration(None=SphereMask(A=WorldPosition().XYZ, B=ParticlePositionWS(), Radius=Add(A=@n67, B=@n68.WorldPosition_SphereMaskRadius control)) {attenuationRadius=65, hardnessPercent=50}) {name=WorldPosition_SphereMask, nodeColor=<rgba 0.0312 0 1 1>}
  n69 = Constant() {r=64}
  n70 = TextureCoordinate()
  n71 = Multiply(
   A=TextureSampleParameter2D(UVs=Add(A=@n52, B=@n70)) {parameterName=Base_mask, texture=/Game/Effects/Fort_Effects/Textures/Masks/T_Square_Gradient_M}.R
   B=@n59
  )
  n72 = Multiply(
   A=Multiply(A=@n62.A, B=MaterialFunctionCall(Distance=@n69) {materialFunction=&Near_Camera_Fade}.Result)
   B=Saturate(None=Multiply(A=@n71, B=@n34.MaskMultiply))
  )
  n73 = Add(
   A=TextureCoordinate() {uTiling=0.5, vTiling=0.5}
   B=AppendVector(
    A=ScalarParameter() {defaultValue=0.25, parameterName=TextureMask_AddU}
    B=ScalarParameter() {defaultValue=0.48, parameterName=TextureMask_AddV}
   )
  )
  n74 = ScalarParameter() {parameterName=LinearMask_AddU}
  n75 = ScalarParameter() {parameterName=LinearMask_AddV}
  n76 = ComponentMask(None=Add(A=TextureCoordinate(), B=AppendVector(A=@n74, B=@n75))) {g=true}
  n77 = Saturate(None=Add(
    A=TextureSample(UVs=@n73) {samplerSource=SSM_Clamp_WorldGroupSettings, texture=/Game/WaterAndWind/Textures/WeaponFireVectorMask}.R
    B=OneMinus(None=@n76)
   ))
  n78 = Multiply(A=Multiply(A=@n77, B=@n77) {constB=2}) {constB=3}
  n79 = LinearInterpolate(A=@n72, B=Multiply(A=@n72, B=Saturate(None=@n78)), Alpha=ScalarParameter() {parameterName=ButtomFade})
  n80 = DepthFade(Opacity=@n79, FadeDistance=@n34.DepthFade)
  n81 = ShadingPathSwitch(Default=@n80, Deferred=@n80, Forward=@n79, Mobile=@n79)
  n82 = OneMinus(None=MaterialFunctionCall(Distance=ScalarParameter() {defaultValue=256, parameterName="Distance (S)"}) {materialFunction=&Near_Camera_Fade}.Result)
  n83 = LinearInterpolate(Alpha=Multiply(A=Multiply(A=@n81, B=@n82), B=@n34.Refraction)) {constA=1, constB=1.05}
  n84 = Constant() {r=1}
  n85 = QualitySwitch(Default=@n83, Low=@n84, High=@n83, Medium=@n84, Epic=@n83)
  ShadingPathSwitch(Default=@n85, Deferred=@n85, Forward=@n84, Mobile=@n84)
  n86 = Constant3Vector()
  n87 = QualitySwitch(Default=@n33.RGB, Low=@n86, High=@n33.RGB, Medium=@n86, Epic=@n33.RGB)
  ShadingPathSwitch(Default=@n87, Deferred=@n87, Forward=@n86, Mobile=@n86)
  n88 = Multiply(A=@n81, B=NamedRerouteUsage())
  n89 = ScalarParameter() {defaultValue=1, parameterName=Use_WorldPositionSphereMask}
  n90 = ScalarParameter() {parameterName="FadeSensitivity (optional) (S)"}
  n91 = TextureCoordinate()
  n92 = DynamicParameter() {parameterIndex=3, paramNames [WorldPosition_SphereMaskRadius control, Param2, Param3, AppearMask]}
  n93 = Multiply(A=Add(A=ComponentMask(None=@n91) {g=true}, B=LinearInterpolate(Alpha=@n92.AppearMask) {constA=-1}) {constB=0}) {constB=3}
  n94 = Multiply(
   A=Multiply(
    A=LinearInterpolate(A=@n88, B=@n81, Alpha=@n89)
    B=MaterialFunctionCall("FadeSensitivity (optional)"=@n90) {materialFunction=&FadeBasedOnViewAngle}.Result
   )
   B=LinearInterpolate(B=Saturate(None=@n93), Alpha=ScalarParameter() {parameterName=UseAppearMask}) {constA=1}
  )
  StaticSwitchParameter(True=Multiply(A=NamedRerouteUsage(), B=@n94), False=@n94) {group=Flow, parameterName=PremultFlowTexture}
 ]
}
