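# Translucent, unlit, two-sided example material for Niagara mesh particles that computes an offset to stretch the mesh toward a supplied position, as in a mesh trail.
# Dynamic parameter 1 supplies an XYZ target, minus particle world position, clamped symmetrically by ClampRange.
# A lerp blends from zero to that offset using U raised to BlendPower times WPOIntensity, so displacement grows along the mesh's UV, apparently for World Position Offset.
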Material /Game/Effects/Niagara/MasterMaterials/MeshTrail/M_Mesh_InterpolateOvertime_Example
graph {
 expressions {
  MaterialExpressionClamp_0 {
   class=MaterialExpressionClamp
   inputs [
    None {from=MaterialExpressionSubtract_0, output=""}
    Min {from=MaterialExpressionMultiply_0, output=""}
    Max {from=MaterialExpressionReroute_0, output=""}
   ]
   properties {maxDefault=30, minDefault=-30}
  }
  MaterialExpressionComponentMask_0 {class=MaterialExpressionComponentMask, inputs [None {from=MaterialExpressionTextureCoordinate_0}], properties {r=true}}
  MaterialExpressionConstant_0 {class=MaterialExpressionConstant, properties {r=0.5}}
  MaterialExpressionConvert_0 {
   class=MaterialExpressionConvert
   inputs [
    None {from=MaterialExpressionDynamicParameter_0, output=X}
    None {from=MaterialExpressionDynamicParameter_0, output=X}
    None {from=MaterialExpressionDynamicParameter_0, output=X}
   ]
   properties {
    nodeName=Make Vector3
    convertInputs [
     {
      type=Scalar
      expressionInput {
       expression=&self:MaterialExpressionDynamicParameter_0
       inputName=None
       mask=1
       maskA=0
       maskB=0
       maskG=0
       maskR=1
       outputIndex=0
      }
     }
     {
      type=Scalar
      expressionInput {
       expression=&self:MaterialExpressionDynamicParameter_0
       inputName=None
       mask=1
       maskA=0
       maskB=0
       maskG=1
       maskR=0
       outputIndex=1
      }
     }
     {
      type=Scalar
      expressionInput {
       expression=&self:MaterialExpressionDynamicParameter_0
       inputName=None
       mask=1
       maskA=0
       maskB=1
       maskG=0
       maskR=0
       outputIndex=2
      }
     }
    ]
    convertOutputs [{type=Vector3}]
   }
  }
  MaterialExpressionDynamicParameter_0 {class=MaterialExpressionDynamicParameter, inputs [], properties {parameterIndex=1, paramNames [X,Y,Z,WPOIntensity]}}
  MaterialExpressionLinearInterpolate_0 {
   class=MaterialExpressionLinearInterpolate
   inputs [
    A {from=""}
    B {from=MaterialExpressionClamp_0}
    Alpha {from=MaterialExpressionMultiply_1}
   ]
  }
  MaterialExpressionMultiply_0 {
   class=MaterialExpressionMultiply
   inputs [
    A {from=""}
    B {from=MaterialExpressionScalarParameter_0}
   ]
   properties {constA=-1}
  }
  MaterialExpressionMultiply_1 {
   class=MaterialExpressionMultiply
   inputs [
    A {from=MaterialExpressionDynamicParameter_0, output=WPOIntensity}
    B {from=MaterialExpressionPower_0}
   ]
  }
  MaterialExpressionParticlePositionWS_0 {class=MaterialExpressionParticlePositionWS, inputs [], properties {}}
  MaterialExpressionPower_0 {
   class=MaterialExpressionPower
   inputs [
    Base {from=MaterialExpressionComponentMask_0}
    Exp {from=MaterialExpressionScalarParameter_1}
   ]
   properties {constExponent=2.5}
  }
  MaterialExpressionReroute_0 {class=MaterialExpressionReroute, inputs [None {from=MaterialExpressionScalarParameter_0}]}
  MaterialExpressionScalarParameter_0 {class=MaterialExpressionScalarParameter, properties {defaultValue=150, parameterName=ClampRange}}
  MaterialExpressionScalarParameter_1 {class=MaterialExpressionScalarParameter, properties {defaultValue=3, parameterName=BlendPower}}
  MaterialExpressionSubtract_0 {
   class=MaterialExpressionSubtract
   inputs [
    A {from=MaterialExpressionConvert_0}
    B {from=MaterialExpressionParticlePositionWS_0}
   ]
  }
  MaterialExpressionTextureCoordinate_0 {class=MaterialExpressionTextureCoordinate}
 }
 properties {
  bAutomaticallySetUsageInEditor=false
  bUsedWithNiagaraMeshParticles=true
  blendMode=BLEND_Translucent
  previewMesh=/Engine/Transient.MaterialEditorMeshComponent_2
  shadingModel=MSM_Unlit
  twoSided=true
  usedShadingModels=Unlit
 }
}
