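# Unlit, two-sided translucent Niagara material for punch flame trails on mesh particles and ribbons.
# Opacity combines a smoke-distorted panning mask, blurred noise, UV end fades, noise-varied view-angle fade, particle alpha and depth fade.
# Emissive blends black with red-to-hot-orange swirls through color noise panned by a dynamic Offset parameter, optionally flattened, then tinted by particle color.
# The noise that varies the angle fade also scales a normal-based offset for vertex displacement.
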Material /Game/Effects/Niagara/MasterMaterials/MeshTrail/M_PunchFlameTrail
graph {
 properties {
  bAutomaticallySetUsageInEditor=false
  bUsedWithNiagaraMeshParticles=true
  bUsedWithNiagaraRibbons=true
  blendMode=BLEND_Translucent
  pixelDepthOffsetMode=PDOM_Legacy
  shadingModel=MSM_Unlit
  twoSided=true
  usedShadingModels=Unlit
 }
 nodes [
  n1 = ParticleColor()
  n2 = Add(A=TextureCoordinate() {uTiling=0.25, vTiling=2}, B=NamedRerouteUsage()) {constB=0}
  n3 = VectorParameter() {parameterName=UV_Scale, channelNames {a=UV_Speed_V, b=UV_Speed_U, g="", r=""}, defaultValue=<rgba 0 0 -1 0>}
  n4 = AppendVector(A=@n3.UV_Speed_U, B=@n3.UV_Speed_U)
  n5 = ComponentMask(None=TextureCoordinate()) {r=true}
  n6 = Power(
   Base=TextureSampleParameter2D(UVs=Panner(Coordinate=@n2, Speed=@n4) {speedX=-1}) {parameterName=Mask, samplerType=SAMPLERTYPE_LinearColor, texture=/Game/Effects/Fort_Effects/BluGlow/Violentium/Violentium_Mask}.R
   Exp=Add(A=LinearInterpolate(Alpha=@n5) {constB=4}) {constB=-1.5}
  )
  n7 = VectorParameter() {parameterName=UV_Scale, channelNames {a=UV_Speed_V, b=UV_Speed_U, g="", r=""}, defaultValue=<rgba 0 0 -1 0>}
  n8 = Panner(Coordinate=TextureCoordinate() {uTiling=0.4, vTiling=2}, Speed=AppendVector(A=@n7.UV_Speed_U, B=@n7.UV_Speed_U)) {speedX=-1}
  n9 = Multiply(
   A=Saturate(None=LinearInterpolate(Alpha=@n6) {constA=-1, constB=2})
   B=LinearInterpolate(
    A=ScalarParameter() {parameterName=MaskNoiseMin}
    Alpha=TextureSample(UVs=@n8) {samplerType=SAMPLERTYPE_Grayscale, texture=&LowResBlurredNoise}.R
   ) {constB=2}
  )
  n10 = TextureCoordinate()
  n11 = Power(Base=OneMinus(None=ComponentMask(None=@n10) {r=true}))
  n12 = TextureCoordinate()
  n13 = Multiply(A=Power(Base=ComponentMask(None=@n12) {r=true})) {constB=10}
  n14 = ScalarParameter() {defaultValue=1, parameterName=AngleFadeMin}
  n15 = ScalarParameter() {defaultValue=4, parameterName=AngleFadeMax}
  n16 = VectorParameter() {parameterName=UV_Scale, channelNames {a=UV_Speed_V, b=UV_Speed_U, g="", r=""}, defaultValue=<rgba 0 0 -1 0>}
  n17 = TextureSampleParameter2D(UVs=Panner(Coordinate=TextureCoordinate() {uTiling=0.7, vTiling=2}, Speed=AppendVector(A=@n16.UV_Speed_U, B=@n16.UV_Speed_U)) {speedX=-1}) {parameterName=WPONoise_Tex, texture=/Game/Weapons/FORT_Melee/Pickaxe_MadameMoth/FX/T_Fibre_Noise}
  n18 = Multiply(
   A=Multiply(A=Multiply(A=@n9, B=@n11), B=Saturate(None=@n13))
   B=MaterialFunctionCall("FadeSensitivity (optional)"=LinearInterpolate(A=@n14, B=@n15, Alpha=@n17.R) {constA=1, constB=4}) {materialFunction=&FadeBasedOnViewAngle}.Result
  )
  DepthFade(Opacity=Multiply(A=@n1.A, B=Multiply(A=@n18) {constB=2}), FadeDistance=ScalarParameter() {parameterName=FadeDistance})
  n19 = Constant3Vector() {constant=<rgba 1 0.0167 0 1>}
  n20 = Constant3Vector() {constant=<rgba 10 0.9364 0.1124 1>}
  n21 = VectorParameter() {parameterName=UV_Scale, channelNames {a=UV_Speed_V, b=UV_Speed_U, g="", r=""}, defaultValue=<rgba 0 0 -1 0>}
  n22 = Panner(
   Coordinate=TextureCoordinate() {uTiling=0.25, vTiling=2}
   Time=DynamicParameter() {defaultValue=<rgba 0.2 1 1 1>, paramNames [Offset,Param2,Param3,Param4]}.Offset
   Speed=AppendVector(A=@n21.UV_Speed_U, B=@n21.UV_Speed_U)
  ) {speedX=-1}
  n23 = TextureSample(UVs=@n22) {texture=/Game/Weapons/WeaponSkins/Textures/T_Swirly_Noise_highContrast}
  n24 = TextureSampleParameter2D(UVs=@n22) {
   parameterName=ColorNoise
   samplerType=SAMPLERTYPE_LinearGrayscale
   texture=/Game/Effects/Fort_Effects/Emotes/UltraLight/Texture/T_PlasmaCoils
  }
  n25 = LinearInterpolate(A=ScalarParameter() {defaultValue=-0.1, parameterName=ColorNoiseLerpMin}, Alpha=Power(Base=@n24.R)) {constA=-0.1, constB=2}
  n26 = Multiply(A=LinearInterpolate(A=Constant3Vector() {constant=<rgba 0 0 0 1>}, B=LinearInterpolate(A=@n19, B=@n20, Alpha=@n23.R), Alpha=Saturate(None=@n25))) {constB=10}
  n27 = Constant() {r=1}
  n28 = ScalarParameter() {parameterName=Remove noise}
  MaterialFunctionCall(Emissive=Multiply(A=LinearInterpolate(A=@n26, B=@n27, Alpha=@n28), B=@n1.RGB)) {materialFunction=&MF_FX_EmissiveAdjustments}
  n29 = VectorParameter() {parameterName=Noise_Scale, channelNames {a=UV_Speed_V, b=UV_Speed_U, g="", r=""}, defaultValue=<rgba 0 0 -0.5 0>}
  n30 = Panner(Coordinate=TextureCoordinate() {uTiling=0.5}, Speed=AppendVector(A=@n29.UV_Speed_U, B=@n29.UV_Speed_U)) {speedX=-0.5}
  NamedRerouteDeclaration(None=Multiply(
    A=TextureSample(UVs=@n30) {texture=&T_Inky_Smoke_Tile}.R
    B=ScalarParameter() {defaultValue=0.5, parameterName=Noise Intensity}
   ) {constB=0.3}) {name=Disturb, nodeColor=<rgba 1 0.0752 0 1>}
  n31 = VertexNormalWS()
  n32 = LinearInterpolate(Alpha=@n17.R) {constB=10}
  Reroute(None=Multiply(A=Multiply(A=@n31, B=@n32), B=ScalarParameter() {defaultValue=1, parameterName=WPOIntensity}))
  ScalarParameter() {defaultValue=-1, parameterName=MaskLerpMin}
  ScalarParameter() {defaultValue=2, parameterName=MaskLerpMax}
 ]
}
